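; Burst gameplay cue that plays effects and sound when a corruption cube floor launches a player.
; It treats the cue's ability level as a cube variant index, picking one of three colors for the first Cascade particle's Color parameter.
; That particle is rotated to face along the cue normal.
; It spawns one sound attached to the target, passing the variant as CubeIndex, and another at the target's location.
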(fn OnBurst(Cascade)
  (params (MyTarget :type Actor%20Object%20Reference) (Parameters :type Gameplay%20Cue%20Parameters%20Structure%20%28by%20ref%29) (ParticleComponents :type Array%20of%20Cascade%20Particle%20System%20Component%20Object%20References) (AudioComponents :type Array%20of%20Audio%20Component%20Object%20References) (BurstCameraShakeInstance :type Camera%20Shake%20Base%20Object%20Reference) (BurstDecalInstance :type Decal%20Actor%20Object%20Reference))
  (bind (_normalizedmagnitude _rawmagnitude _effectcontext _matchedtagname _originaltag _aggregatedsourcetags _aggregatedtargettags _location _normal _instigator _effectcauser _sourceobject _physicalmaterial _gameplayeffectlevel _abilitylevel) (Ability|GameplayCue|BreakGameplayCueParameters Parameters))
  (bind _output (Utilities|Array|Get(acopy) ParticleComponents))
  (bind _rootcomponent (Variables|/Script/Engine.Actor|GetRootComponent MyTarget))
  (Utilities|FlowControl|Sequence
    (:then_0
      (Effects|Components|ParticleSystem|SetVectorParameter _output "Color" (Math|Conversions|ToVector(LinearColor) (Utilities|Array|Get(acopy) (Utilities|Array|MakeArray "(R=0.041667,G=0.815359,B=1.000000,A=1.000000)" "(R=0.824663,G=0.041384,B=1.000000,A=1.000000)" "(R=1.000000,G=0.363283,B=0.017294,A=1.000000)") _abilitylevel))) ; Color change for blue/gold cube variants
      (Transformation|SetWorldRotation _output (Math|Conversions|RotationFromXVector _normal) false false) ; Orient to cube surface
    )
    (:then_1
      (bind _returnvalue (Audio|SpawnSoundAttached (Variables|Default|GetSoundPlayerAttached) _rootcomponent))
      (bind _returnvalue_2 (Audio|SpawnSoundatLocation 0 (Variables|Default|GetSoundAtLocation) (Transformation|GetWorldLocation _rootcomponent)))
      (Audio|Parameter|SetIntegerParameter _returnvalue "CubeIndex" _abilitylevel :!class /Script/AudioExtensions.AudioParameterControllerInterface)
      (return))))
