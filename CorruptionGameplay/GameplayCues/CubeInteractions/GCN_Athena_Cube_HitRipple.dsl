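; Gameplay cue for the hit ripple on Cube interactions, tuning its burst particles and audio to the hit.
; Raw magnitude is remapped to particle size parameters and bucketed into an integer audio Strength by summing overlapping range checks.
; Ability level acts as the cube variant index, choosing one of three particle colors and setting CubeIndex on an additional ripple sound spawned at the cue location.
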(fn OnBurst(Cascade)
  (params (MyTarget :type Actor%20Object%20Reference) (Parameters :type Gameplay%20Cue%20Parameters%20Structure%20%28by%20ref%29) (ParticleComponents :type Array%20of%20Cascade%20Particle%20System%20Component%20Object%20References) (AudioComponents :type Array%20of%20Audio%20Component%20Object%20References) (BurstCameraShakeInstance :type Camera%20Shake%20Base%20Object%20Reference) (BurstDecalInstance :type Decal%20Actor%20Object%20Reference))
  (bind (_normalizedmagnitude _rawmagnitude _effectcontext _matchedtagname _originaltag _aggregatedsourcetags _aggregatedtargettags _location _normal _instigator _effectcauser _sourceobject _physicalmaterial _gameplayeffectlevel _abilitylevel) (Ability|GameplayCue|BreakGameplayCueParameters Parameters))
  (bind _rawmagnitude_2 (LocalVariables|GetRawMagnitude))
  (bind _returnvalue (Math|Float|MapRangeClamped _rawmagnitude_2 0.3 2.25 0.0 1.0))
  (bind _output (Utilities|Array|Get(acopy) ParticleComponents))
  (LocalVariables|SetRawMagnitude _rawmagnitude)
  (Utilities|FlowControl|Sequence
    (:then_0
      (Effects|Components|ParticleSystem|SetFloatParameter _output "Size" _returnvalue) ; Scale particles/lifetime based on damage
      (Effects|Components|ParticleSystem|SetVectorParameter _output "SizeV" (Math|Conversions|ToVector(Float) _returnvalue)))
    (:then_1
      (Audio|Parameter|SetIntegerParameter (Utilities|Array|Get(acopy) AudioComponents) "Strength" (Math|Integer|int+int (Math|Conversions|ToInteger(Boolean) (Math|Float|InRange(Float) (LocalVariables|GetRawMagnitude) 0.0 2.5)) (Math|Conversions|ToInteger(Boolean) (Math|Float|InRange(Float) (LocalVariables|GetRawMagnitude) 1.0 2.5)) (Math|Conversions|ToInteger(Boolean) (Math|Float|InRange(Float) (LocalVariables|GetRawMagnitude) 2.0 2.5))) :!class /Script/AudioExtensions.AudioParameterControllerInterface))
    (:then_2
      (Effects|Components|ParticleSystem|SetVectorParameter (Utilities|Array|Get(acopy) ParticleComponents) "Color" (Math|Conversions|ToVector(LinearColor) (Utilities|Array|Get(acopy) (Utilities|Array|MakeArray "(R=0.041667,G=0.815359,B=1.000000,A=1.000000)" "(R=0.824663,G=0.041384,B=1.000000,A=1.000000)" "(R=1.000000,G=0.363283,B=0.017294,A=1.000000)") _abilitylevel))) ; Color change for blue/gold cube variants
      (bind _returnvalue_2 (Audio|SpawnSoundatLocation 0 "/CorruptionGameplay/Gameplay/Environmental/Cube/Audio/Cube_HitRipple_Cue.Cube_HitRipple_Cue" _location))
      (Audio|Parameter|SetIntegerParameter _returnvalue_2 "CubeIndex" _abilitylevel :!class /Script/AudioExtensions.AudioParameterControllerInterface)
      (return))))
